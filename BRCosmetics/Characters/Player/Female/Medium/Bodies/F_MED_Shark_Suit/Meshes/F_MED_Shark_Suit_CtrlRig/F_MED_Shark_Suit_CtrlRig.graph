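# Control Rig for the female medium Shark Suit player cosmetic that procedurally sways the suit's tail chain.
# A single Chain Harmonics unit runs from Begin Execution on the chain rooted at C_Root_Main_Root_Jnt under the pelvis.
# Only its sine-eased, slightly noisy wave is active, with pendulum and reach disabled and debug drawing left on.
# The fin, pouch and other dyn bones in the hierarchy are not driven by this graph.
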ControlRigBlueprint /BRCosmetics/Characters/Player/Female/Medium/Bodies/F_MED_Shark_Suit/Meshes/F_MED_Shark_Suit_CtrlRig/F_MED_Shark_Suit_CtrlRig
graphs [
 RigVMFunctionLibrary {links [], nodes []}
 RigVMModel {nodes [
   n1 = &RigUnit_BeginExecution()
   /Script/ControlRig.RigUnit_ChainHarmonics(
    ChainRoot=C_Root_Main_Root_Jnt
    DrawWorldOffset=(Rotation=(X=0,Y=0,Z=0,W=1),Translation=(X=0,Y=0,Z=0),Scale3D=(X=1,Y=1,Z=1))
    Pendulum=(bEnabled=false,PendulumStiffness=2,PendulumGravity=(X=0,Y=0,Z=0),PendulumBlend=0.75,PendulumDrag=0.98,PendulumMaximum=0.1,UnwindAxis=(X=0,Y=1,Z=0),UnwindMinimum=0.2,UnwindMaximum=0.05)
    Reach=(bEnabled=false,ReachTarget=(X=0,Y=1,Z=0),ReachAxis=(X=1,Y=0,Z=0),ReachMinimum=0,ReachMaximum=1,ReachEase=QuadraticEaseInOut)
    Speed=(X=1,Y=1.6,Z=1)
    Wave=(bEnabled=True,WaveFrequency=(X=1,Y=1,Z=0.8),WaveAmplitude=(X=0,Y=2.5,Z=0),WaveOffset=(X=0,Y=15,Z=0),WaveNoise=(X=0,Y=0.1,Z=0),WaveMaximum=0.5,WaveEase=SineEaseIn)
    WaveCurve=(EditorCurveData=(Keys=((),(Time=1,Value=1)),DefaultValue=340282346638528859811704183484516925440,PreInfinityExtrap=RCCE_Constant,PostInfinityExtrap=RCCE_Constant))
    bDrawDebug=True
    ExecutePin=@n1.ExecutePin
   ) {is=struct RigVMUnitNode}
   RigUnit_BeginExecution = &RigUnit_BeginExecution()
  ]}
]
hierarchy [
 root {loc=<0 0 0>, rot=<0 0 0 1>}
 attach {parent=root, loc=<0 0 0>, rot=<0 0 0 1>}
 pelvis {parent=root, loc=<0 -0.5301 98.0571>, rot=<0.0127 0.707 -0.0127 -0.707>}
 spine_01 {parent=pelvis, loc=<6.1208 -0.9328 0>, rot=<0 0 0.034 -0.9994>}
 spine_02 {parent=spine_01, loc=<6.5996 -0.2095 0>, rot=<0 0 0.0238 -0.9997>}
 spine_03 {parent=spine_02, loc=<6.4914 -0.1237 0>, rot=<0 0 0.0149 -0.9999>}
 spine_04 {parent=spine_03, loc=<6.9598 0.0794 0>, rot=<0 0 -0.0132 -0.9999>}
 spine_05 {parent=spine_04, loc=<6.9987 0 0>, rot=<0 0 -0.0015 -1>}
 clavicle_l {parent=spine_05, loc=<5.4831 -0.1053 -3.2946>, rot=<0.0465 -0.7289 -0.009 -0.6829>}
 upperarm_l {parent=clavicle_l, loc=<9.8658 -1.2136 -0.4104>, rot=<0.0164 -0.3245 -0.0269 -0.9454>}
 lowerarm_l {parent=upperarm_l, loc=<26.465 -0.8345 0.259>, rot=<0.0036 0.0161 -0.2298 -0.9731>}
 hand_l {parent=lowerarm_l, loc=<22.7355 -0.3203 -0.6105>, rot=<-0.6784 -0.0223 0.0735 -0.7307>}
 index_metacarpal_l {parent=hand_l, loc=<1.6535 0.0183 -2.96>, rot=<0.0158 -0.0124 -0.0173 -0.9996>}
 index_01_l {parent=index_metacarpal_l, loc=<8.3008 0.0357 -0.0524>, rot=<-0.0044 0.0485 0.0924 -0.9945>}
 index_02_l {parent=index_01_l, loc=<3.3587 -0.3003 0.0452>, rot=<-0.002 0.0027 0.3225 -0.9466>}
 index_03_l {parent=index_02_l, loc=<2.5578 -0.123 -0.0252>, rot=<-0.002 0.0067 0.2558 -0.9667>}
 middle_metacarpal_l {parent=hand_l, loc=<2.1856 -0.1725 -0.5984>, rot=<0.006 0.0168 -0.0329 -0.9993>}
 middle_01_l {parent=middle_metacarpal_l, loc=<7.3275 -0.2447 0.0129>, rot=<-0.0208 0.0042 0.0858 -0.9961>}
 middle_02_l {parent=middle_01_l, loc=<4.0123 0.1402 0.1175>, rot=<-0.0218 0.0258 0.3512 -0.9357>}
 middle_03_l {parent=middle_02_l, loc=<2.6077 0.0534 0.1373>, rot=<0.0082 0.0188 0.1708 -0.9851>}
 pinky_metacarpal_l {parent=hand_l, loc=<1.7616 -0.9753 3.3814>, rot=<-0.0397 0.0814 -0.069 -0.9935>}
 pinky_01_l {parent=pinky_metacarpal_l, loc=<7.1709 -0.4831 -0.1094>, rot=<-0.0654 -0.0035 0.2242 -0.9723>}
 pinky_02_l {parent=pinky_01_l, loc=<2.8274 0.0892 -0.1043>, rot=<0.0037 0.0011 0.2854 -0.9584>}
 pinky_03_l {parent=pinky_02_l, loc=<2.2046 0.0308 -0.0078>, rot=<-0.0002 0.0098 0.1445 -0.9895>}
 ring_metacarpal_l {parent=hand_l, loc=<2.2074 -0.3625 1.3638>, rot=<-0.0284 0.0722 -0.0458 -0.9959>}
 ring_01_l {parent=ring_metacarpal_l, loc=<7.0295 -0.6197 -0.1446>, rot=<-0.0489 -0.0252 0.1892 -0.9804>}
 ring_02_l {parent=ring_01_l, loc=<3.6174 0.4647 -0.1027>, rot=<0 0 0.2806 -0.9598>}
 ring_03_l {parent=ring_02_l, loc=<2.4152 0.114 0.0199>, rot=<0 0 0.1189 -0.9929>}
 thumb_01_l {parent=hand_l, loc=<1.3118 -0.3191 -2.459>, rot=<0.6546 -0.4042 0.0664 -0.6354>}
 thumb_02_l {parent=thumb_01_l, loc=<5.9336 -0.0448 0.0401>, rot=<-0.0037 0.0213 0.298 -0.9543>}
 thumb_03_l {parent=thumb_02_l, loc=<2.2223 0.3682 -0.0907>, rot=<0.0118 -0.013 0.0066 -0.9998>}
 dyn_l_armFin_01 {parent=hand_l, loc=<-1.0641 2.6928 0.1858>, rot=<0.6743 0.0579 -0.0529 -0.7343>}
 dyn_l_armFin_02 {parent=dyn_l_armFin_01, loc=<7.6863 0 0>, rot=<-0.0148 -0.0361 -0.0021 -0.9992>}
 dyn_l_armFin_03 {parent=dyn_l_armFin_02, loc=<9.8871 0 0>, rot=<0.0717 0.3608 0.1812 -0.9121>}
 weapon_l {parent=hand_l, loc=<-0.1974 -3.4111 0.3811>, rot=<0 0 0 1>}
 lowerarm_twist_01_l {parent=lowerarm_l, loc=<18.316 0 0>, rot=<0 0 0 1>}
 lowerarm_twist_02_l {parent=lowerarm_l, loc=<10.6707 0 0>, rot=<0 0 0 1>}
 upperarm_twist_01_l {parent=upperarm_l, loc=<7.2135 -0.1134 0>, rot=<0 0 0 -1>}
 upperarm_twist_02_l {parent=upperarm_l, loc=<19.7998 -0.1134 0>, rot=<0 0 0 -1>}
 clavicle_r {parent=spine_05, loc=<5.4831 -0.1053 3.2946>, rot=<0.7289 0.0465 -0.6829 0.009>}
 upperarm_r {parent=clavicle_r, loc=<-9.8658 1.2136 0.4104>, rot=<0.0164 -0.3245 -0.0269 -0.9454>}
 lowerarm_r {parent=upperarm_r, loc=<-26.4651 0.8345 -0.259>, rot=<0.0036 0.0161 -0.2298 -0.9731>}
 hand_r {parent=lowerarm_r, loc=<-22.7356 0.3203 0.6106>, rot=<-0.6784 -0.0223 0.0735 -0.7307>}
 index_metacarpal_r {parent=hand_r, loc=<-1.6534 -0.0181 2.9601>, rot=<0.0158 -0.0124 -0.0173 -0.9996>}
 index_01_r {parent=index_metacarpal_r, loc=<-8.3008 -0.0357 0.0524>, rot=<-0.0044 0.0485 0.0924 -0.9945>}
 index_02_r {parent=index_01_r, loc=<-3.3589 0.3003 -0.0453>, rot=<-0.002 0.0027 0.3225 -0.9466>}
 index_03_r {parent=index_02_r, loc=<-2.5578 0.123 0.0252>, rot=<-0.002 0.0067 0.2558 -0.9667>}
 middle_metacarpal_r {parent=hand_r, loc=<-2.1856 0.1726 0.5984>, rot=<0.006 0.0168 -0.0329 -0.9993>}
 middle_01_r {parent=middle_metacarpal_r, loc=<-7.3273 0.2447 -0.0129>, rot=<-0.0208 0.0042 0.0858 -0.9961>}
 middle_02_r {parent=middle_01_r, loc=<-4.0124 -0.14 -0.1175>, rot=<-0.0218 0.0258 0.3512 -0.9357>}
 middle_03_r {parent=middle_02_r, loc=<-2.6077 -0.0533 -0.1373>, rot=<0.0082 0.0188 0.1708 -0.9851>}
 pinky_metacarpal_r {parent=hand_r, loc=<-1.7617 0.9753 -3.3814>, rot=<-0.0397 0.0814 -0.069 -0.9935>}
 pinky_01_r {parent=pinky_metacarpal_r, loc=<-7.1707 0.4832 0.1094>, rot=<-0.0654 -0.0035 0.2242 -0.9723>}
 pinky_02_r {parent=pinky_01_r, loc=<-2.8274 -0.0892 0.1043>, rot=<0.0037 0.0011 0.2854 -0.9584>}
 pinky_03_r {parent=pinky_02_r, loc=<-2.2045 -0.031 0.0078>, rot=<-0.0002 0.0098 0.1445 -0.9895>}
 ring_metacarpal_r {parent=hand_r, loc=<-2.2073 0.3627 -1.3638>, rot=<-0.0284 0.0722 -0.0458 -0.9959>}
 ring_01_r {parent=ring_metacarpal_r, loc=<-7.0297 0.6197 0.1446>, rot=<-0.0489 -0.0252 0.1892 -0.9804>}
 ring_02_r {parent=ring_01_r, loc=<-3.6172 -0.4648 0.1027>, rot=<0 0 0.2806 -0.9598>}
 ring_03_r {parent=ring_02_r, loc=<-2.4153 -0.1141 -0.0199>, rot=<0 0 0.1189 -0.9929>}
 thumb_01_r {parent=hand_r, loc=<-1.3119 0.3192 2.459>, rot=<0.6546 -0.4042 0.0664 -0.6354>}
 thumb_02_r {parent=thumb_01_r, loc=<-5.9336 0.0447 -0.0401>, rot=<-0.0037 0.0213 0.298 -0.9543>}
 thumb_03_r {parent=thumb_02_r, loc=<-2.2223 -0.3682 0.0907>, rot=<0.0118 -0.013 0.0066 -0.9998>}
 dyn_r_armFin_01 {parent=hand_r, loc=<0.6421 -2.3444 -0.2235>, rot=<-0.0497 0.6694 0.7404 -0.0354>}
 dyn_r_armFin_02 {parent=dyn_r_armFin_01, loc=<6.9486 0 0>, rot=<-0.0011 -0.0028 -0.002 -1>}
 dyn_r_armFin_03 {parent=dyn_r_armFin_02, loc=<10.2212 0 0>, rot=<-0.1753 -0.9215 -0.0648 0.3405>}
 weapon_r {parent=hand_r, loc=<0.1975 3.4111 -0.3811>, rot=<0 0 0 1>}
 lowerarm_twist_01_r {parent=lowerarm_r, loc=<-18.3161 0 0>, rot=<0 0 0 1>}
 lowerarm_twist_02_r {parent=lowerarm_r, loc=<-10.6707 0 0>, rot=<0 0 0 1>}
 upperarm_twist_01_r {parent=upperarm_r, loc=<-7.2135 0.1134 0>, rot=<0 0 0 1>}
 upperarm_twist_02_r {parent=upperarm_r, loc=<-19.7998 0.1134 0>, rot=<0 0 0 1>}
 neck_01 {parent=spine_05, loc=<11.3138 0.1529 0>, rot=<0 0 -0.2003 -0.9797>}
 neck_02 {parent=neck_01, loc=<4.5627 0 0>, rot=<0 0 0.0504 -0.9987>}
 head {parent=neck_02, loc=<5.3914 0 0>, rot=<0 0 0.115 -0.9934>}
 base_r_chestPouch {parent=spine_04, loc=<5.3295 10.0659 11.1015>, rot=<0.0294 -0.7065 -0.0294 -0.7065>}
 dyn_r_chestPouch {parent=base_r_chestPouch, loc=<0 0 0>, rot=<0 0 0 1>}
 base_l_chestPouch {parent=spine_04, loc=<5.0542 10.0429 -11.3218>, rot=<0.0294 -0.7065 -0.0294 -0.7065>}
 dyn_l_chestPouch {parent=base_l_chestPouch, loc=<0 0 0>, rot=<0 0 0 1>}
 thigh_l {parent=pelvis, loc=<-1.0855 -0.2109 -8.7053>, rot=<0.0009 0.0355 0.0155 -0.9993>}
 calf_l {parent=thigh_l, loc=<-42.5284 -0.0019 -0.7297>, rot=<0.0006 -0.0161 -0.0391 -0.9991>}
 calf_twist_01_l {parent=calf_l, loc=<-35.8483 -1.9197 -0.5794>, rot=<0.0237 0.0095 -0.0035 -0.9997>}
 calf_twist_02_l {parent=calf_l, loc=<-19.1546 -1.7943 -0.2638>, rot=<0.0237 0.0095 -0.0035 -0.9997>}
 foot_l {parent=calf_l, loc=<-46.7401 -2.5014 -0.592>, rot=<0.0427 -0.0238 0.0353 -0.9982>}
 ball_l {parent=foot_l, loc=<-3.5658 10.8066 -0.1544>, rot=<0.0001 -0.0001 -0.7063 -0.7079>}
 dyn_l_foot_mouth_01 {parent=ball_l, loc=<-0.1285 -1.4979 1.3963>, rot=<-0.0072 0.0449 0.2173 -0.975>}
 dyn_l_foot_mouth_02 {parent=dyn_l_foot_mouth_01, loc=<9.5656 0 0>, rot=<-0.3817 -0.5953 0.3817 -0.5953>}
 dyn_l_foot_r_fin_01 {parent=foot_l, loc=<-3.8547 6.1293 8.6067>, rot=<-0.0357 -0.7045 -0.0284 -0.7082>}
 dyn_l_foot_l_fin_01 {parent=foot_l, loc=<-3.7674 7.5765 -6.5171>, rot=<-0.0357 -0.7045 -0.0284 -0.7082>}
 dyn_l_footTailFin_01 {parent=foot_l, loc=<-0.6677 -6.7511 -0.3334>, rot=<-0.0357 -0.7045 -0.0284 -0.7082>}
 dyn_l_footTailFin_02 {parent=dyn_l_footTailFin_01, loc=<0 -4.4268 0>, rot=<0 0 0 1>}
 thigh_twist_01_l {parent=thigh_l, loc=<-27.952 0 0>, rot=<0.0013 0 0 -1>}
 dyn_l_legSuit_01 {parent=thigh_l, loc=<-34.4492 -1.6457 -0.5713>, rot=<-0.0019 -0.7317 0.0015 -0.6816>}
 deform_groin_base_l {parent=thigh_l, loc=<0.4801 5.6776 -0.2486>, rot=<0 0 0 1>}
 deform_groin_l {parent=deform_groin_base_l, loc=<0 0 0>, rot=<0 0 0 1>}
 thigh_r {parent=pelvis, loc=<-1.0853 -0.2109 8.7053>, rot=<0.0355 -0.0009 0.9993 0.0155>}
 calf_r {parent=thigh_r, loc=<42.5286 0.0019 0.7297>, rot=<0.0006 -0.0161 -0.0391 -0.9991>}
 calf_twist_01_r {parent=calf_r, loc=<35.8485 1.9197 0.5794>, rot=<0.0237 0.0095 -0.0035 -0.9997>}
 calf_twist_02_r {parent=calf_r, loc=<19.1546 1.7943 0.2638>, rot=<0.0237 0.0095 -0.0035 -0.9997>}
 foot_r {parent=calf_r, loc=<46.7402 2.5014 0.592>, rot=<0.0427 -0.0238 0.0353 -0.9982>}
 ball_r {parent=foot_r, loc=<3.5658 -10.8066 0.1545>, rot=<0.0001 -0.0001 -0.7063 -0.7079>}
 dyn_r_foot_mouth_01 {parent=ball_r, loc=<0.1258 1.4982 -1.3665>, rot=<-0.0449 -0.0072 0.975 0.2173>}
 dyn_r_foot_mouth_02 {parent=dyn_r_foot_mouth_01, loc=<9.5655 0 0>, rot=<-0.3817 -0.5953 0.3817 -0.5953>}
 dyn_r_foot_l_fin_01 {parent=foot_r, loc=<4.0106 -6.6614 -8.4093>, rot=<0.7082 0.0284 -0.7045 -0.0357>}
 dyn_r_foot_r_fin_01 {parent=foot_r, loc=<3.594 -7.618 6.7939>, rot=<0.7082 0.0284 -0.7045 -0.0357>}
 dyn_r_footTailFin_01 {parent=foot_r, loc=<0.6673 6.7484 0.3632>, rot=<0.7082 0.0284 -0.7045 -0.0357>}
 dyn_r_footTailFin_02 {parent=dyn_r_footTailFin_01, loc=<0 -4.4268 0>, rot=<0 0 0 1>}
 thigh_twist_01_r {parent=thigh_r, loc=<27.952 0 0>, rot=<0.0013 0 0 -1>}
 dyn_r_legSuit_01 {parent=thigh_r, loc=<34.4418 1.6458 0.4642>, rot=<-0.6816 0.0015 0.7317 0.0019>}
 dyn_r_kneepad {parent=thigh_r, loc=<42.7634 -4.4725 0.0171>, rot=<-0.6816 0.0015 0.7317 0.0019>}
 deform_groin_base_r {parent=thigh_r, loc=<-0.4799 -5.6776 0.2485>, rot=<0 0 0 1>}
 deform_groin_r {parent=deform_groin_base_r, loc=<0 0 0>, rot=<0 0 0 1>}
 tail_attach {parent=pelvis, loc=<-5.843 -8.8751 0>, rot=<-0.0127 -0.707 0.0127 -0.707>}
 C_Root_Main_Root_Jnt {parent=pelvis, loc=<-5.8428 -8.8753 0>, rot=<-0.2341 -0.6672 0.2341 -0.6672>}
 C_Tail_A_Base_Jnt {parent=C_Root_Main_Root_Jnt, loc=<0 0 0>, rot=<0.5 -0.5 -0.5 0.5>}
 C_Tail_A_1_Jnt {parent=C_Tail_A_Base_Jnt, loc=<10.9994 0 0>, rot=<1 0 0 0>}
 C_Tail_A_2_Jnt {parent=C_Tail_A_1_Jnt, loc=<11.0006 0 0>, rot=<0 0 0 -1>}
 C_Tail_A_3_Jnt {parent=C_Tail_A_2_Jnt, loc=<11.0001 0 0>, rot=<0 0 0 -1>}
 C_Tail_A_4_Jnt {parent=C_Tail_A_3_Jnt, loc=<10.9997 0 0>, rot=<0 0 0 -1>}
 C_Tail_A_5_Jnt {parent=C_Tail_A_4_Jnt, loc=<10.9999 0 0>, rot=<0 0 0 -1>}
 deform_glute_base_l {parent=pelvis, loc=<-7.3466 -9.0213 -9.1245>, rot=<0.0172 0.031 -0.4477 -0.8935>}
 deform_glute_l {parent=deform_glute_base_l, loc=<0 0 0>, rot=<0 0 0 1>}
 deform_glute_base_r {parent=pelvis, loc=<-7.3464 -9.0213 9.1245>, rot=<0.031 -0.0172 0.8935 -0.4477>}
 deform_glute_r {parent=deform_glute_base_r, loc=<0 0 0>, rot=<0 0 0 1>}
 ik_foot_root {parent=root, loc=<0 0 0>, rot=<0 0 0 1>}
 ik_foot_l {parent=ik_foot_root, loc=<14.8262 -7.3545 8.296>, rot=<-0.0357 -0.7045 -0.0284 0.7082>}
 ik_foot_r {parent=ik_foot_root, loc=<-14.8262 -7.3546 8.2958>, rot=<-0.7082 -0.0284 0.7045 -0.0357>}
 ik_hand_root {parent=root, loc=<0 0 0>, rot=<0 0 0 1>}
 ik_hand_gun {parent=ik_hand_root, loc=<-48.1744 3.3546 104.4381>, rot=<-0.6276 -0.1309 -0.4183 0.6434>}
 ik_hand_l {parent=ik_hand_gun, loc=<59.3272 67.6997 34.3524>, rot=<-0.6158 -0.7027 -0.3565 0>}
 ik_hand_r {parent=ik_hand_gun, loc=<0 0 0>, rot=<0 0 0 -1>}
]
