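# Physics asset for the left arm fin of the female medium Shark Suit body cosmetic, giving the fin slight secondary motion.
# Two fin capsules form a chain from the left-hand capsule, with tight swing and even tighter twist limits at each joint.
# The hand is angularly locked to a pelvis box, so only the fin joints are set up to flex.
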PhysicsAsset /BRCosmetics/Characters/Player/Female/Medium/Bodies/F_MED_Shark_Suit/Meshes/F_MED_Shark_Suit_Physics_L_Arm
bodies {
 dyn_l_armFin_01 [{shapeType=Capsule, length=7, radius=1.6, shapeName=dyn_l_armFin_01_capsule, center=<4 0 0>, rotation=<rot -90 0 0>}]
 dyn_l_armFin_02 [{shapeType=Capsule, length=8, radius=1.6, shapeName=dyn_l_armFin_02_capsule, center=<8 0 0>, rotation=<rot -90 0 0>}]
 hand_l [{shapeType=Capsule, length=14.8101, radius=5.4583, center=<8.3001 -3.2763 0.0001>, rotation=<rot 90 -12.3391 -19.6739>}]
 pelvis [{shapeType=Box, extentX=15.3576, extentY=9.0976, extentZ=28.3007}]
}
constraints [
 {
  bone1Name=hand_l
  bone2Name=pelvis
  swing1LimitDegrees=45
  swing1Motion=Locked
  swing2LimitDegrees=45
  swing2Motion=Locked
  twistLimitDegrees=45
  twistMotion=Locked
 }
 {bone1Name=dyn_l_armFin_01, bone2Name=hand_l, swing1LimitDegrees=4, swing2LimitDegrees=8, twistLimitDegrees=1}
 {bone1Name=dyn_l_armFin_02, bone2Name=dyn_l_armFin_01, swing1LimitDegrees=4, swing2LimitDegrees=8, twistLimitDegrees=1}
]
