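# Physics asset for the slippers of the female medium Shark Suit cosmetic, covering the shark mouth, tail fin and side fins on each foot.
# Each dynamic slipper bone has a box body constrained to its foot with small swing and twist limits, each favoring one swing axis.
# Both feet are box bodies with angular motion locked to a pelvis box, anchoring the wobbling parts.
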PhysicsAsset /BRCosmetics/Characters/Player/Female/Medium/Bodies/F_MED_Shark_Suit/Meshes/F_MED_Shark_Suit_Physics_Slippers
bodies {
 dyn_l_footTailFin_01 [{shapeType=Box, extentX=1.5174, extentY=5, extentZ=7.8761, center=<0 -2.8303 0>}]
 dyn_l_foot_l_fin_01 [{shapeType=Box, extentX=5, extentY=5, extentZ=1.9987, center=<2.2004 0 -1.4587>, rotation=<rot -33.5414 0 0>}]
 dyn_l_foot_mouth_01 [{shapeType=Box, extentX=9.5093, extentY=1.8566, extentZ=11.568, center=<5.1328 0 0>}]
 dyn_l_foot_r_fin_01 [{shapeType=Box, extentX=5, extentY=5, extentZ=1.9987, center=<-2.1633 0 -1.6288>, rotation=<rot 36.9763 0 0>}]
 dyn_r_footTailFin_01 [{shapeType=Box, extentX=1.5174, extentY=5, extentZ=7.8761, center=<0 -2.8303 0>}]
 dyn_r_foot_l_fin_01 [{shapeType=Box, extentX=4.0024, extentY=4.9637, extentZ=2, center=<1.6878 0 -1.1466>, rotation=<rot -38.8836 0 0>}]
 dyn_r_foot_mouth_01 [{shapeType=Box, extentX=9.5093, extentY=1.8566, extentZ=11.568, center=<5.1328 0 0>}]
 dyn_r_foot_r_fin_01 [{shapeType=Box, extentX=4.063, extentY=4.1478, extentZ=2, center=<-1.8459 0 -1.4312>, rotation=<rot 34.3092 0 0>}]
 foot_l [{shapeType=Box, extentX=5, extentY=9.463, extentZ=9.3672, center=<-1.1254 4.2069 0.2073>}]
 foot_r [{shapeType=Box, extentX=5, extentY=9.463, extentZ=9.3672, center=<1.1273 -4.1765 -0.5415>}]
 pelvis [{shapeType=Box, extentX=15.3576, extentY=9.0976, extentZ=28.3007}]
}
constraints [
 {bone1Name=dyn_r_foot_mouth_01, bone2Name=foot_r, swing1LimitDegrees=10, swing2LimitDegrees=3, twistLimitDegrees=3}
 {bone1Name=dyn_r_footTailFin_01, bone2Name=foot_r, swing1LimitDegrees=20, swing2LimitDegrees=5, twistLimitDegrees=5}
 {bone1Name=dyn_l_foot_mouth_01, bone2Name=foot_l, swing1LimitDegrees=10, swing2LimitDegrees=3, twistLimitDegrees=3}
 {bone1Name=dyn_l_foot_r_fin_01, bone2Name=foot_l, swing1LimitDegrees=3, swing2LimitDegrees=20, twistLimitDegrees=3}
 {bone1Name=dyn_l_foot_l_fin_01, bone2Name=foot_l, swing1LimitDegrees=3, swing2LimitDegrees=20, twistLimitDegrees=3}
 {bone1Name=dyn_l_footTailFin_01, bone2Name=foot_l, swing1LimitDegrees=20, swing2LimitDegrees=5, twistLimitDegrees=5}
 {
  bone1Name=foot_l
  bone2Name=pelvis
  swing1LimitDegrees=45
  swing1Motion=Locked
  swing2LimitDegrees=45
  swing2Motion=Locked
  twistLimitDegrees=45
  twistMotion=Locked
 }
 {
  bone1Name=foot_r
  bone2Name=pelvis
  swing1LimitDegrees=45
  swing1Motion=Locked
  swing2LimitDegrees=45
  swing2Motion=Locked
  twistLimitDegrees=45
  twistMotion=Locked
 }
 {bone1Name=dyn_r_foot_r_fin_01, bone2Name=foot_r, swing1LimitDegrees=3, swing2LimitDegrees=20, twistLimitDegrees=3}
 {bone1Name=dyn_r_foot_l_fin_01, bone2Name=foot_r, swing1LimitDegrees=3, swing2LimitDegrees=20, twistLimitDegrees=3}
]
