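# Physics asset for the tail of the female medium Shark Suit player cosmetic, defining collision and joint limits for simulating the tail.
# A six-segment chain of identical-sized capsules runs from the tail base through joint 5, anchored to a larger pelvis capsule.
# Each link is constrained to its parent with tight twist and first-swing limits and a wide second swing, so the tail bends mostly in a single plane.
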PhysicsAsset /BRCosmetics/Characters/Player/Female/Medium/Bodies/F_MED_Shark_Suit/Meshes/F_MED_Shark_Suit_Physics_Tail
bodies {
 C_Tail_A_1_Jnt [{shapeType=Capsule, length=3.4362, radius=4, center=<6.2855 0 0>, rotation=<rot -90 0.0008 -0.0008>}]
 C_Tail_A_2_Jnt [{shapeType=Capsule, length=3.4362, radius=4, center=<6.2855 0 0>, rotation=<rot -90 0.0008 -0.0008>}]
 C_Tail_A_3_Jnt [{shapeType=Capsule, length=3.4362, radius=4, center=<6.2855 0 0>, rotation=<rot -90 0.0008 -0.0008>}]
 C_Tail_A_4_Jnt [{shapeType=Capsule, length=3.4362, radius=4, center=<6.2855 0 0>, rotation=<rot -90 0.0008 -0.0008>}]
 C_Tail_A_5_Jnt [{shapeType=Capsule, length=3.4362, radius=4, center=<6.2855 0 0>, rotation=<rot -90 0.0008 -0.0008>}]
 C_Tail_A_Base_Jnt [{shapeType=Capsule, length=3.4362, radius=4, center=<6.2855 0 0>, rotation=<rot -90 -0.0003 0.0005>}]
 pelvis [{shapeType=Capsule, length=12.6667, radius=10.48}]
}
constraints [
 {bone1Name=C_Tail_A_5_Jnt, bone2Name=C_Tail_A_4_Jnt, swing1LimitDegrees=3, swing2LimitDegrees=45, twistLimitDegrees=3}
 {bone1Name=C_Tail_A_Base_Jnt, bone2Name=pelvis, swing1LimitDegrees=3, swing2LimitDegrees=45, twistLimitDegrees=3}
 {bone1Name=C_Tail_A_1_Jnt, bone2Name=C_Tail_A_Base_Jnt, swing1LimitDegrees=3, swing2LimitDegrees=45, twistLimitDegrees=3}
 {bone1Name=C_Tail_A_2_Jnt, bone2Name=C_Tail_A_1_Jnt, swing1LimitDegrees=3, swing2LimitDegrees=45, twistLimitDegrees=3}
 {bone1Name=C_Tail_A_3_Jnt, bone2Name=C_Tail_A_2_Jnt, swing1LimitDegrees=3, swing2LimitDegrees=45, twistLimitDegrees=3}
 {bone1Name=C_Tail_A_4_Jnt, bone2Name=C_Tail_A_3_Jnt, swing1LimitDegrees=3, swing2LimitDegrees=45, twistLimitDegrees=3}
]
