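# Animation Blueprint for the female medium Shark Suit cosmetic body, adding secondary motion to the parent mesh's copied pose.
# A deform Control Rig feeds crouch-switched tail and slipper RigidBody sims, then targeting-switched arm sims.
# A second Control Rig applies when not moving, followed by kneepad and chest-pouch constraints, with pouches rotated by clavicle-driven alphas.
# On customization load, variants lacking the Mat2 tag toggle a material hide parameter for long backblings.
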AnimBlueprint /BRCosmetics/Characters/Player/Female/Medium/Bodies/F_MED_Shark_Suit/Meshes/F_MED_Shark_Suit_AnimBP
target_skeleton=&/F_MED_Shark_Suit_Skeleton
graphs [
 AnimGraph {nodes [
   n1 = RigidBody(ComponentPose=Local To Component(LocalPose=Use cached pose 'Tail_Dynamics'())) {
    bClampLinearTranslationLimitToRefPose=true
    bOverrideWorldGravity=true
    overridePhysicsAsset=&/F_MED_Shark_Suit_Physics_L_Arm
    simulationSpace=BaseBoneSpace
    alphaBoolBlend {blendOption=Linear}
    baseBoneRef {boneName=hand_l}
    componentAppliedLinearAccClamp=<2000 2000 2000>
    componentLinearAccScale=<0.4 0.4 0.4>
    componentLinearVelScale=<0.1 0.1 0.1>
    overrideWorldGravity {z=-980}
   }
   n2 = RigidBody(ComponentPose=@n1) {
    bClampLinearTranslationLimitToRefPose=true
    bOverrideWorldGravity=true
    overridePhysicsAsset=&/F_MED_Shark_Suit_Physics_R_Arm
    simulationSpace=BaseBoneSpace
    alphaBoolBlend {blendOption=Linear}
    baseBoneRef {boneName=hand_r}
    componentAppliedLinearAccClamp=<2000 2000 2000>
    componentLinearAccScale=<0.4 0.4 0.4>
    componentLinearVelScale=<0.1 0.1 0.1>
    overrideWorldGravity {z=-980}
   }
   n3 = RigidBody(ComponentPose=Local To Component(LocalPose=Use cached pose 'Tail_Dynamics'()), Alpha=0.9) {
    bClampLinearTranslationLimitToRefPose=true
    bOverrideWorldGravity=true
    overridePhysicsAsset=&F_MED_Shark_Suit_Physics_L_Arm
    simulationSpace=BaseBoneSpace
    alphaBoolBlend {blendOption=Linear}
    baseBoneRef {boneName=hand_l}
    componentAppliedLinearAccClamp=<2000 2000 2000>
    componentLinearAccScale=<0.4 0.4 0.4>
    componentLinearVelScale=<0.1 0.1 0.1>
    overrideWorldGravity {z=-980}
   }
   n4 = RigidBody(ComponentPose=@n3, Alpha=0.9) {
    bClampLinearTranslationLimitToRefPose=true
    bOverrideWorldGravity=true
    overridePhysicsAsset=&F_MED_Shark_Suit_Physics_R_Arm
    simulationSpace=BaseBoneSpace
    alphaBoolBlend {blendOption=Linear}
    baseBoneRef {boneName=hand_r}
    componentAppliedLinearAccClamp=<2000 2000 2000>
    componentLinearAccScale=<0.4 0.4 0.4>
    componentLinearVelScale=<0.1 0.1 0.1>
    overrideWorldGravity {z=-980}
   }
   n5 = "Transform (Modify) Bone - Bone: dyn_r_armFin_01"(ComponentPose=@n4, Alpha=1, Translation="0,0,0", Rotation="-5,0,0", Scale="1,1,1") {
    rotationMode=BMM_Additive
    rotationSpace=BCS_BoneSpace
    alphaBoolBlend {blendOption=Linear}
    boneToModify {boneName=dyn_r_armFin_01}
    rotation {pitch=-5}
   }
   n6 = "Transform (Modify) Bone - Bone: dyn_r_armFin_01"(ComponentPose=@n5, Alpha=1, Translation="0,0,0", Rotation="-5,0,0", Scale="1,1,1") {
    rotationMode=BMM_Additive
    rotationSpace=BCS_BoneSpace
    alphaBoolBlend {blendOption=Linear}
    boneToModify {boneName=dyn_r_armFin_01}
    rotation {pitch=-5}
   }
   n7 = Get bIsMoving2D()
   n8 = Control Rig(
    Source=Blend Poses by bool(BlendPose_0=Component To Local(ComponentPose=@n6), BlendPose_1=Component To Local(ComponentPose=@n2), bActiveValue=Get bIsTargeting()) {blendType=Linear}
    Alpha="To Float (Boolean)"(InBool=NOT Boolean(A=@n7))
   ) {
    alphaBoolBlend {blendOption=Linear}
    controlRigAssetReference {blueprintRigClass=/BRCosmetics/Characters/Player/Female/Medium/Bodies/F_MED_Shark_Suit/Meshes/F_MED_Shark_Suit_CtrlRig/F_MED_Shark_Suit_CtrlRig.F_MED_Shark_Suit_CtrlRig_C}
   }
   n9 = "Constraint - dyn_r_kneepad (PP)"!(ComponentPose=Local To Component(LocalPose=@n8), Alpha=1, ConstraintWeights_0=0.5, ConstraintWeights_1:float=0.5) -> (Pose) {
    alphaBoolBlend {blendOption=Linear}
    boneToModify {boneName=dyn_r_kneepad}
    constraintSetup [{targetBone {boneName=calf_r}}, {targetBone {boneName=thigh_r}}]
    constraintWeights [1,1]
   }
   n10 = "Constraint - base_r_chestPouch (PPPPP)"!(
    ComponentPose=@n9
    Alpha=1
    ConstraintWeights_0=0.6764
    ConstraintWeights_1:float=0.1523
    ConstraintWeights_2:float=0.1261
    ConstraintWeights_3:float=0.0448
    ConstraintWeights_4:float=0.0003
   ) -> (Pose) {
    alphaBoolBlend {blendOption=Linear}
    boneToModify {boneName=base_r_chestPouch}
    constraintSetup [
     {targetBone {boneName=spine_04}}
     {targetBone {boneName=spine_05}}
     {targetBone {boneName=clavicle_r}}
     {targetBone {boneName=upperarm_twist_01_r}}
     {targetBone {boneName=spine_03}}
    ]
    constraintWeights [1,1,1,1,1]
   }
   n11 = "Constraint - base_l_chestPouch (PPPPP)"!(
    ComponentPose=@n10
    Alpha=1
    ConstraintWeights_0=0.5323
    ConstraintWeights_1:float=0.1991
    ConstraintWeights_2:float=0.1898
    ConstraintWeights_3:float=0.078
    ConstraintWeights_4:float=0.0009
   ) -> (Pose) {
    alphaBoolBlend {blendOption=Linear}
    boneToModify {boneName=base_l_chestPouch}
    constraintSetup [
     {targetBone {boneName=spine_04}}
     {targetBone {boneName=clavicle_l}}
     {targetBone {boneName=spine_05}}
     {targetBone {boneName=upperarm_twist_01_l}}
     {targetBone {boneName=spine_03}}
    ]
    constraintWeights [1,1,1,1,1]
   }
   n12 = "Transform (Modify) Bone - Bone: base_l_chestPouch"(ComponentPose=@n11, Alpha=Get clavicle_Up_left() -> (clavicle_Up_left:double), Translation="0,0,0", Rotation="0,15,0", Scale="1,1,1") {
    rotationMode=BMM_Additive
    rotationSpace=BCS_BoneSpace
    scaleMode=BMM_Additive
    scaleSpace=BCS_BoneSpace
    translationMode=BMM_Additive
    translationSpace=BCS_BoneSpace
    alphaBoolBlend {blendOption=Linear}
    boneToModify {boneName=base_l_chestPouch}
    rotation {yaw=15}
   }
   n13 = "Transform (Modify) Bone - Bone: base_r_chestPouch"(ComponentPose=@n12, Alpha=Get clavicle_Up_right() -> (clavicle_Up_right:double), Translation="0,0,0", Rotation="0,15,0", Scale="1,1,1") {
    rotationMode=BMM_Additive
    rotationSpace=BCS_BoneSpace
    scaleMode=BMM_Additive
    scaleSpace=BCS_BoneSpace
    translationMode=BMM_Additive
    translationSpace=BCS_BoneSpace
    alphaBoolBlend {blendOption=Linear}
    boneToModify {boneName=base_r_chestPouch}
    rotation {yaw=15}
   }
   Output Pose(Result=Component To Local(ComponentPose=@n13)) {layerGroup=None}
   n14 = RigidBody(ComponentPose=Local To Component(LocalPose=Use cached pose 'controlRig_cache'())) {
    bOverrideWorldGravity=true
    overridePhysicsAsset=&/F_MED_Shark_Suit_Physics
    simulationSpace=BaseBoneSpace
    alphaBoolBlend {blendOption=Linear}
    baseBoneRef {boneName=root}
    componentLinearAccScale=<0.35 0.35 0.45>
    overrideWorldGravity {z=-980}
   }
   n15 = RigidBody(ComponentPose=@n14, Alpha=0.5) {
    overridePhysicsAsset=&/F_MED_Shark_Suit_Physics_Slippers
    simulationSpace=BaseBoneSpace
    alphaBoolBlend {blendOption=Linear}
    baseBoneRef {boneName=root}
    componentLinearAccScale=<0.35 0.35 0.45>
    overrideWorldGravity {z=-980}
   }
   n16 = RigidBody(ComponentPose=@n15, Alpha=0.75) {
    bOverrideWorldGravity=true
    overridePhysicsAsset=&/F_MED_Shark_Suit_Physics_Tail
    simulationSpace=BaseBoneSpace
    alphaBoolBlend {blendOption=Linear}
    baseBoneRef {boneName=root}
    componentLinearAccScale=<0.35 0.35 0.45>
    overrideWorldGravity {z=-980}
   }
   n17 = RigidBody(ComponentPose=Local To Component(LocalPose=Use cached pose 'controlRig_cache'()), Alpha=0.5) {
    overridePhysicsAsset=&F_MED_Shark_Suit_Physics_Slippers
    simulationSpace=BaseBoneSpace
    alphaBoolBlend {blendOption=Linear}
    baseBoneRef {boneName=root}
    componentLinearAccScale=<0.35 0.35 0.45>
    overrideWorldGravity {z=-980}
   }
   n18 = RigidBody(ComponentPose=@n17, Alpha=0.3) {
    bOverrideWorldGravity=true
    overridePhysicsAsset=&F_MED_Shark_Suit_Physics_Tail
    simulationSpace=BaseBoneSpace
    alphaBoolBlend {blendOption=Linear}
    baseBoneRef {boneName=root}
    componentLinearAccScale=<0.35 0.35 0.45>
    overrideWorldGravity {z=-980}
   }
   n19 = "Transform (Modify) Bone - Bone: C_Root_Main_Root_Jnt"(ComponentPose=@n18, Alpha=1, Translation="0,0,0", Rotation="0,0,3", Scale="1,1,1") {
    rotationMode=BMM_Additive
    translationMode=BMM_Additive
    alphaBoolBlend {blendOption=Linear}
    boneToModify {boneName=C_Root_Main_Root_Jnt}
    rotation {roll=3}
   }
   n20 = "Transform (Modify) Bone - Bone: C_Tail_A_1_Jnt"(ComponentPose=@n19, Alpha=1, Translation="0,0,0", Rotation="0,0,2", Scale="1,1,1") {
    rotationMode=BMM_Additive
    translationMode=BMM_Additive
    alphaBoolBlend {blendOption=Linear}
    boneToModify {boneName=C_Tail_A_1_Jnt}
    rotation {roll=2}
   }
   n21 = Get bIsCrouching()
   Save cached pose 'Tail_Dynamics'(Pose=Blend Poses by bool(BlendPose_0=Component To Local(ComponentPose=@n16), BlendPose_1=Component To Local(ComponentPose=@n20), bActiveValue=NOT Boolean(A=@n21)) {blendType=Linear}) {cacheName=Tail_Dynamics}
   n22 = Control Rig(Source=Copy Pose From Mesh() {bUseAttachedParent=true}) {
    alphaBoolBlend {blendOption=Linear}
    controlRigAssetReference {blueprintRigClass=&/F_MED_DeformRig_Shark_Suit_ControlRig.F_MED_DeformRig_Shark_Suit_ControlRig_C}
   }
   Save cached pose 'controlRig_cache'(Pose=@n22) {cacheName=controlRig_cache}
  ]}
 EventGraph {nodes [
   Event BlueprintUpdateAnimation()
   > n1 = Sequence()
   > n2 = Gravity Override State Function DEPRECATED(CharacterPartAnimInstance=Self-Reference(), GameTimeInSeconds=GetGameTimeInSeconds(), GravityOverrideParams=Get Gravity Override Params())
   > Set GravityOverride(GravityOverride=@n2)
   > Set is crouching alpha value(is crouching alpha value:double="To Float (Boolean)"(InBool=Get bIsCrouching())) -> (Output_Get:double)
   Set SkelMesh(execute=@n1.then_1, SkelMesh=GetPartSkeletalMeshComponent())
   > Set clavicle_Up_left(clavicle_Up_left:double=GetAlphaFromReferenceTransformDelta(
     SkeletalMeshComponent=Get SkelMesh()
     BoneName=clavicle_l
     BaseName=spine_04
     DeltaComponentType=EulerY
     InputRange=(X=0,Y=60)
     TwistAxis="0,0,1"
    )) -> (Output_Get:double)
   > Set clavicle_Up_right(clavicle_Up_right:double=GetAlphaFromReferenceTransformDelta(
     SkeletalMeshComponent=Get SkelMesh()
     BoneName=clavicle_r
     BaseName=spine_04
     DeltaComponentType=EulerY
     InputRange=(X=0,Y=60)
     TwistAxis="0,0,1"
    )) -> (Output_Get:double)
   Event OnCharacterCustomizationLoaded()
   > n3 = Get Dynamic Material Instance(SkeletalMeshComponent=GetPartSkeletalMeshComponent())
   > n4 = Branch(Condition=CharacterItemDefinitionHasAnyMetaTag(Tags=Make Array("[0]"=(TagName="Cosmetics.Variant.Property.Mat2"))))
   n5 = CheckForBackbling()
   Set isWearingBackbling(execute=@n4.else, isWearingBackbling=@n5)
   > Set isWearingCape(isWearingCape=@n5.bIsBackblingLong)
   > n6 = Branch(Condition=Get isWearingCape() -> (isWearingCape:bool))
   > SetScalarParameterValue(self=@n3, ParameterName=Hide Element 09, Value=1)
   SetScalarParameterValue(execute=@n6.else, self=@n3, ParameterName=Hide Element 09, Value=0)
  ]}
 OnNewFunction {nodes [
   OnNewFunction!() -> (then, Context:"Anim Update Context Structure (by ref)", Node:"Anim Node Reference Structure (by ref)") {is=title K2Node_FunctionEntry}
   > Return Node()
  ]}
]
variables [
 BodyMesh {type=SkeletalMeshComponent}
 cape and tail combo {type=boolean}
 characterHasTail {type=boolean}
 clavicle_Up_left {type=number}
 clavicle_Up_right {type=number}
 Gravity Override Params {type=GravityOverrideParamsStruct}
 GravityOverride {type=Vector}
 is crouching alpha value {type=number}
 isWearingBackbling {type=boolean}
 isWearingCape {type=boolean}
 PawnOwner {type=PlayerPawn_Athena_C}
 SkelMesh {type=SkeletalMeshComponent}
 TailMesh {type=SkeletalMeshComponent}
]
