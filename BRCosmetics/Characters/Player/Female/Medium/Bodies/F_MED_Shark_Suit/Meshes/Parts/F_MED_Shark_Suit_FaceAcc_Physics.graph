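# Physics asset for the face accessory of the female medium Shark Suit cosmetic, giving simulation bodies to its jaws, eleven teeth and zipper.
# Each tooth capsule hinges off the jaws with tight angular limits on two axes and a wider third, with dyn_teeth_01 swapping its wide axis from swing2 to twist.
# The zipper box is angularly locked to the jaws, while the jaws-to-neck_01 and neck_01-to-spine_05 joints rotate freely.
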PhysicsAsset /BRCosmetics/Characters/Player/Female/Medium/Bodies/F_MED_Shark_Suit/Meshes/Parts/F_MED_Shark_Suit_FaceAcc_Physics
bodies {
 dyn_jaws [{shapeType=Capsule, length=2, radius=2}]
 dyn_teeth_01 [{shapeType=Capsule, length=2, radius=2, center=<0 0 -3.0086>}]
 dyn_teeth_010 [{shapeType=Capsule, length=2, radius=2, center=<0 0 -3.4822>}]
 dyn_teeth_011 [{shapeType=Capsule, length=2, radius=2, center=<0 0 -3.2292>}]
 dyn_teeth_02 [{shapeType=Capsule, length=2, radius=2, center=<0 0.0001 -3.1373>}]
 dyn_teeth_03 [{shapeType=Capsule, length=2, radius=2, center=<0 0.0001 -3.1373>}]
 dyn_teeth_04 [{shapeType=Capsule, length=2, radius=2, center=<0 0.0001 -3.1373>}]
 dyn_teeth_05 [{shapeType=Capsule, length=2, radius=2, center=<0 0.0001 -3.1373>}]
 dyn_teeth_06 [{shapeType=Capsule, length=2, radius=2, center=<0 0.0001 -3.1373>}]
 dyn_teeth_07 [{shapeType=Capsule, length=2, radius=2, center=<0 0.0001 -3.1373>}]
 dyn_teeth_08 [{shapeType=Capsule, length=2, radius=2, center=<-0.0002 0 -2.9364>}]
 dyn_teeth_09 [{shapeType=Capsule, length=2, radius=2, center=<-0.0001 0 -3.0986>}]
 dyn_zipper_offset [{shapeType=Box, extentX=3.3441, extentY=2, extentZ=0.5, center=<2.613 0.0001 -0.985>}]
 neck_01 [{shapeType=Capsule, length=18.708, radius=4.8091, center=<-1.9821 -4.3335 0.0001>, rotation=<rot 0 90 0>}]
 spine_05 [{shapeType=Capsule, length=12.7082, radius=6.8412, center=<9.7357 -3.0141 -0.0001>, rotation=<rot 0 90 0>}]
}
constraints [
 {bone1Name=dyn_teeth_07, bone2Name=dyn_jaws, swing1LimitDegrees=2, swing2LimitDegrees=45, twistLimitDegrees=7}
 {bone1Name=dyn_teeth_06, bone2Name=dyn_jaws, swing1LimitDegrees=2, swing2LimitDegrees=45, twistLimitDegrees=7}
 {bone1Name=dyn_teeth_05, bone2Name=dyn_jaws, swing1LimitDegrees=2, swing2LimitDegrees=45, twistLimitDegrees=7}
 {bone1Name=dyn_teeth_04, bone2Name=dyn_jaws, swing1LimitDegrees=2, swing2LimitDegrees=45, twistLimitDegrees=7}
 {bone1Name=dyn_teeth_03, bone2Name=dyn_jaws, swing1LimitDegrees=2, swing2LimitDegrees=45, twistLimitDegrees=7}
 {bone1Name=dyn_teeth_02, bone2Name=dyn_jaws, swing1LimitDegrees=2, swing2LimitDegrees=45, twistLimitDegrees=7}
 {bone1Name=dyn_teeth_01, bone2Name=dyn_jaws, swing1LimitDegrees=2, swing2LimitDegrees=7, twistLimitDegrees=45}
 {bone1Name=dyn_teeth_08, bone2Name=dyn_jaws, swing1LimitDegrees=2, swing2LimitDegrees=45, twistLimitDegrees=7}
 {bone1Name=dyn_teeth_09, bone2Name=dyn_jaws, swing1LimitDegrees=2, swing2LimitDegrees=45, twistLimitDegrees=7}
 {bone1Name=dyn_teeth_010, bone2Name=dyn_jaws, swing1LimitDegrees=2, swing2LimitDegrees=45, twistLimitDegrees=7}
 {bone1Name=dyn_teeth_011, bone2Name=dyn_jaws, swing1LimitDegrees=2, swing2LimitDegrees=45, twistLimitDegrees=7}
 {
  bone1Name=dyn_zipper_offset
  bone2Name=dyn_jaws
  swing1LimitDegrees=45
  swing1Motion=Locked
  swing2LimitDegrees=45
  swing2Motion=Locked
  twistLimitDegrees=45
  twistMotion=Locked
 }
 {
  bone1Name=dyn_jaws
  bone2Name=neck_01
  swing1LimitDegrees=45
  swing1Motion=Free
  swing2LimitDegrees=45
  swing2Motion=Free
  twistLimitDegrees=45
  twistMotion=Free
 }
 {
  bone1Name=neck_01
  bone2Name=spine_05
  swing1LimitDegrees=45
  swing1Motion=Free
  swing2LimitDegrees=45
  swing2Motion=Free
  twistLimitDegrees=45
  twistMotion=Free
 }
]
