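# Physics asset for the dangling zipper face accessory on the female medium Shark Suit character cosmetic.
# Capsules on spine_05, neck_01, dyn_jaws and dyn_zipper_offset, plus a box on dyn_zipper, form a constrained chain.
# The offset is angularly locked to the jaws, while the jaws-to-neck and neck-to-spine joints rotate freely.
# The zipper joint has one wide swing limit and tight other limits, giving a hinge-like flap.
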PhysicsAsset /BRCosmetics/Characters/Player/Female/Medium/Bodies/F_MED_Shark_Suit/Meshes/Parts/F_MED_Shark_Suit_FaceAcc_Zipper
bodies {
 dyn_jaws [{shapeType=Capsule, length=2, radius=2}]
 dyn_zipper [{shapeType=Box, extentX=3.3044, extentY=0.8011, extentZ=4.269, center=<2.4192 -0.0106 0.0293>, rotation=<rot 0 89.9952 -89.4929>}]
 dyn_zipper_offset [{shapeType=Capsule, length=8.2524, radius=1.6802, center=<2.701 0 -2.3604>, rotation=<rot 0 -89.9998 -90.0371>}]
 neck_01 [{shapeType=Capsule, length=18.708, radius=4.8091, center=<-2.8685 -3.2391 0.0001>, rotation=<rot 0 90 0>}]
 spine_05 [{shapeType=Capsule, length=12.7082, radius=6.8412, center=<8.5837 -2.5729 -0.0001>, rotation=<rot 0 90 0>}]
}
constraints [
 {
  bone1Name=dyn_zipper_offset
  bone2Name=dyn_jaws
  swing1LimitDegrees=45
  swing1Motion=Locked
  swing2LimitDegrees=45
  swing2Motion=Locked
  twistLimitDegrees=45
  twistMotion=Locked
 }
 {
  bone1Name=dyn_jaws
  bone2Name=neck_01
  swing1LimitDegrees=45
  swing1Motion=Free
  swing2LimitDegrees=45
  swing2Motion=Free
  twistLimitDegrees=45
  twistMotion=Free
 }
 {
  bone1Name=neck_01
  bone2Name=spine_05
  swing1LimitDegrees=45
  swing1Motion=Free
  swing2LimitDegrees=45
  swing2Motion=Free
  twistLimitDegrees=45
  twistMotion=Free
 }
 {bone1Name=dyn_zipper, bone2Name=dyn_zipper_offset, swing1LimitDegrees=3, swing2LimitDegrees=120, twistLimitDegrees=2}
]
