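# Post-physics style AnimBlueprint adding secondary motion to the female Shark Suit face accessory.
# It copies the attached parent's pose, constrains dyn_jaws_offset to the head, then chains AnimDynamics jaws, a spine-relative RigidBody, hair AnimDynamics and a head-relative zipper RigidBody, mostly at partial alpha.
# Pouch dynamics, an IsTargeting blend and extra zipper simulations do not reach the output.
# The event graph has no logic.
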AnimBlueprint /BRCosmetics/Characters/Player/Female/Medium/Bodies/F_MED_Shark_Suit/Meshes/Parts/F_MED_Shark_Suit_FaceAcc_AnimBP
target_skeleton=&/F_MED_Shark_Suit_FaceAcc_Skeleton
graphs [
 AnimGraph {nodes [
   n1 = Anim Dynamics - Bone: dyn_l_pouch(Alpha=1, ExternalForce="0,0,0") {
    angularDampingOverride=0.9999
    angularSpringConstant=700
    bAngularSpring=true
    bGravityOverrideInSimSpace=true
    bLinearSpring=true
    bOverrideAngularDamping=true
    bUseGravityOverride=true
    linearSpringConstant=100
    alphaBoolBlend {blendOption=Linear}
    boundBone {boneName=dyn_l_pouch}
    componentAppliedLinearAccClamp=<3000 3000 3000>
    componentLinearAccScale=<0.1 0.1 0.1>
    gravityOverride {z=-20}
    physicsBodyDefinitions [{
      sphereCollisionRadius=0
      boundBone {boneName=dyn_l_pouch}
      boxExtents=<10 10 5>
      constraintSetup {angularLimitsMax=<5 5 5>, angularLimitsMin=<-5 -5 -5>, angularTarget=<0 0 0>}
      localJointOffset=<0 0 -1>
     }]
   }
   Anim Dynamics - Bone: dyn_r_pouch(ComponentPose=@n1, Alpha=1, ExternalForce="0,0,0") {
    angularDampingOverride=0.9999
    angularSpringConstant=700
    bAngularSpring=true
    bGravityOverrideInSimSpace=true
    bLinearSpring=true
    bOverrideAngularDamping=true
    bUseGravityOverride=true
    linearSpringConstant=100
    alphaBoolBlend {blendOption=Linear}
    boundBone {boneName=dyn_r_pouch}
    componentAppliedLinearAccClamp=<3000 3000 3000>
    componentLinearAccScale=<0.1 0.1 0.1>
    gravityOverride {z=-20}
    physicsBodyDefinitions [{
      sphereCollisionRadius=0
      boundBone {boneName=dyn_r_pouch}
      boxExtents=<10 10 5>
      constraintSetup {angularLimitsMax=<5 5 5>, angularLimitsMin=<-5 -5 -5>, angularTarget=<0 0 0>}
      localJointOffset=<1 -1 0.1>
     }]
   }
   n2 = RigidBody(ComponentPose=Local To Component(LocalPose=Copy Pose From Mesh() {bCopyCurves=true, bUseAttachedParent=true})) {
    bOverrideWorldGravity=true
    overridePhysicsAsset=&/F_MED_Shark_Suit_FaceAcc_Zipper
    simulationSpace=BaseBoneSpace
    alphaBoolBlend {blendOption=Linear}
    baseBoneRef {boneName=root}
    componentLinearAccScale=<0.35 0.35 0.45>
    overrideWorldGravity {z=-980}
   }
   Blend Poses by bool(BlendPose_0=Component To Local(ComponentPose=@n2), bActiveValue=Get bIsTargeting()) {blendType=Linear}
   RigidBody() {
    bOverrideWorldGravity=true
    overridePhysicsAsset=&F_MED_Shark_Suit_FaceAcc_Zipper
    simulationSpace=BaseBoneSpace
    alphaBoolBlend {blendOption=Linear}
    baseBoneRef {boneName=root}
    componentLinearAccScale=<0.35 0.35 0.45>
    overrideWorldGravity {z=-980}
   }
   n3 = "Constraint - dyn_jaws_offset (P)"(
    ComponentPose=Local To Component(LocalPose=Copy Pose From Mesh() {bCopyCurves=true, bUseAttachedParent=true})
    Alpha=1
    ConstraintWeights_0=1
   ) {
    alphaBoolBlend {blendOption=Linear}
    boneToModify {boneName=dyn_jaws_offset}
    constraintSetup [{targetBone {boneName=head}}]
    constraintWeights [1]
   }
   n4 = Anim Dynamics - Bone: dyn_jaws(ComponentPose=@n3, Alpha=0.1, ExternalForce="0,0,0") {
    angularDampingOverride=0.9999
    angularSpringConstant=50
    bAngularSpring=true
    bGravityOverrideInSimSpace=true
    bLinearSpring=true
    bOverrideAngularDamping=true
    bOverrideLinearDamping=true
    gravityScale=0.05
    linearDampingOverride=0.9999
    linearSpringConstant=150
    alphaBoolBlend {blendOption=Linear}
    boundBone {boneName=dyn_jaws}
    componentLinearAccScale=<0.25 0.25 0.25>
    physicsBodyDefinitions [{
      sphereCollisionRadius=0
      boundBone {boneName=dyn_jaws}
      boxExtents=<10 20 20>
      constraintSetup {angularLimitsMax=<45 5 45>, angularLimitsMin=<-45 -5 -45>, angularTarget=<0 0 0>, linearAxesMax=<5 5 12>, linearAxesMin=<-5 -5 -12>}
      localJointOffset=<0 0 0.1>
     }]
   }
   n5 = RigidBody!(ComponentPose=@n4, Alpha=0.35, OverrideWorldGravity:Vector="0,0,-500", ExternalForce) -> (Pose) {
    bClampLinearTranslationLimitToRefPose=true
    bOverrideWorldGravity=true
    overridePhysicsAsset=&/F_MED_Shark_Suit_FaceAcc_Physics
    simulationSpace=BaseBoneSpace
    alphaBoolBlend {blendOption=Linear}
    baseBoneRef {boneName=spine_05}
    componentLinearAccScale=<0.45 0.45 0.45>
    overrideWorldGravity {z=-500}
   }
   n6 = Anim Dynamics - Bone: dyn_hairBase(ComponentPose=@n5, Alpha=0.5, ExternalForce="0,0,0") {
    bGravityOverrideInSimSpace=true
    alphaBoolBlend {blendOption=Linear}
    boundBone {boneName=dyn_hairBase}
    physicsBodyDefinitions [{
      sphereCollisionRadius=0
      boundBone {boneName=dyn_hairBase}
      boxExtents=<20 20 20>
      constraintSetup {angularLimitsMax=<0 61.1999 61.1999>, angularLimitsMin=<0 -57.5999 -43.2>, angularTarget=<0 0 0>}
      localJointOffset=<3 0 0>
     }]
   }
   n7 = Anim Dynamics - Bone: dyn_hair_copy_01(ComponentPose=@n6, Alpha=0.2, ExternalForce="0,0,0") {
    bGravityOverrideInSimSpace=true
    alphaBoolBlend {blendOption=Linear}
    boundBone {boneName=dyn_hair_copy_01}
    physicsBodyDefinitions [{
      sphereCollisionRadius=0
      boundBone {boneName=dyn_hair_copy_01}
      boxExtents=<20 20 20>
      constraintSetup {angularLimitsMax=<0 61.1999 61.1999>, angularLimitsMin=<0 -57.5999 -43.2>, angularTarget=<0 0 0>}
      localJointOffset=<3 0 0>
     }]
   }
   n8 = Anim Dynamics - Bone: dyn_hair_copy_02(ComponentPose=@n7, Alpha=0.2, ExternalForce="0,0,0") {
    bGravityOverrideInSimSpace=true
    alphaBoolBlend {blendOption=Linear}
    boundBone {boneName=dyn_hair_copy_02}
    physicsBodyDefinitions [{
      sphereCollisionRadius=0
      boundBone {boneName=dyn_hair_copy_02}
      boxExtents=<20 20 20>
      constraintSetup {angularLimitsMax=<0 61.1999 61.1999>, angularLimitsMin=<0 -57.5999 -43.2>, angularTarget=<0 0 0>}
      localJointOffset=<3 0 0>
     }]
   }
   n9 = Anim Dynamics - Bone: dyn_hairClump(ComponentPose=@n8, Alpha=0.2, ExternalForce="0,0,0") {
    bGravityOverrideInSimSpace=true
    bLinearSpring=true
    linearSpringConstant=500
    alphaBoolBlend {blendOption=Linear}
    boundBone {boneName=dyn_hairClump}
    physicsBodyDefinitions [{
      sphereCollisionRadius=0
      boundBone {boneName=dyn_hairClump}
      boxExtents=<20 20 20>
      constraintSetup {angularLimitsMax=<0 20 20>, angularLimitsMin=<0 -20 -20>, angularTarget=<0 0 0>, linearAxesMax=<0.5 0.5 0.5>, linearAxesMin=<-0.5 -0.5 -0.5>}
      localJointOffset=<0 0 -3>
     }]
   }
   n10 = RigidBody!(ComponentPose=@n9, Alpha=0.9, ExternalForce) -> (Pose) {
    bOverrideWorldGravity=true
    overridePhysicsAsset=&F_MED_Shark_Suit_FaceAcc_Zipper
    simulationSpace=BaseBoneSpace
    alphaBoolBlend {blendOption=Linear}
    baseBoneRef {boneName=head}
    componentLinearAccScale=<0.35 0.35 0.45>
    overrideWorldGravity {z=-980}
   }
   Output Pose(Result=Component To Local(ComponentPose=@n10)) {layerGroup=None}
  ]}
 EventGraph {nodes [Event BlueprintUpdateAnimation()]}
]
variables []
